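# Niagara system for the weapon trail of the VanceGuard pickaxe cosmetic.
# A single Trail emitter spawns short-lived anim-trail particles through the FN anim-trail modules, with a ribbon ID offset, and draws them with two ribbon renderers.
# Curves fade the particles from bright purple-pink toward blue over their life, driving both particle color and a separate CleanTrailColor attribute, each scaled by a time-of-day brightness value.
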NiagaraSystem /BRCosmetics/Weapons/FORT_Melee/Pickaxe_VanceGuard/FX/NS_Pickaxe_Vanceguard_Trail
dependencies {
 usedDataInterfaces [&NiagaraDataInterfaceColorCurve, &NiagaraDataInterfaceCurve, &NiagaraDataInterfaceSocketReader]
 usedDynamicInputs [&FNTimeOfDayColor, &FNTimeOfDayScalar]
 usedModules [
  &self:SystemSpawnScript.SystemScriptSource.SystemScriptGraph.NiagaraNodeAssignment_1.SetVariables_#1
  &self:SystemSpawnScript.SystemScriptSource.SystemScriptGraph.NiagaraNodeAssignment_0.SetVariables_#2
  &SystemState
 ]
}
emitters {Trail {
  emitterSpawnScript [&FN_RibbonIDOffsetUtility()]
  emitterUpdateScript [
   &EmitterState()
   &FN_SpawnAnimTrail()
  ]
  particleSpawnScript [
   &FN_InitializeAnimTrailBoneIndecies()
   &FN_InitializeAnimTrail(Lifetime=0.18)
  ]
  particleUpdateScript [
   &SolveForcesAndVelocity(Clamp Velocity=false, Limit Acceleration=false)
   &FN_UpdateAnimTrail() {module=FN_UpdateAnimTrail001}
   &ScaleColor(
    Scale Alpha=&FloatFromCurve(FloatCurve=DataInterface() {propertyValues {Curve {keys [{interpMode=RCIM_Cubic, tangentMode=RCTM_User, time=0, value=1}, {interpMode=RCIM_Cubic, tangentMode=RCTM_User, time=1, value=0}]}}})
    Scale RGB=&MakeVectorFromLinearColorRGB(Color=&Multiply_LinearColorByFloat(
      Linear Color=&ColorFromCurve(ColorCurve=DataInterface() {propertyValues {
         AlphaCurve {keys [<at 0 1 RCIM_Cubic>,<at 1 0 RCIM_Cubic>]}
         BlueCurve {keys [<at 0 3 RCIM_Cubic>,<at 0.3698 1.5>,<at 0.854 1 RCIM_Cubic>]}
         GreenCurve {keys [<at 0 0.2867 RCIM_Cubic>,<at 0.3698 0.0121>,<at 0.854 0 RCIM_Cubic>]}
         RedCurve {keys [<at 0 2.4589 RCIM_Cubic>,<at 0.3698 0.614>,<at 0.854 0.1094 RCIM_Cubic>]}
        }})
      Float=$System.ToDBrightness
     ))
    ScaleA=true
    ScaleRGB=true
   )
   /SetVariables(Particles.CleanTrailColor:LinearColor=&Multiply_LinearColorByFloat(
     Linear Color=&Multiply_LinearColorByFloat(
      Linear Color=&ColorFromCurve(ColorCurve=DataInterface() {propertyValues {
         AlphaCurve {keys [<at 0 1 RCIM_Cubic>,<at 1 0 RCIM_Cubic>]}
         BlueCurve {keys [<at 0 4 RCIM_Cubic>,<at 0.3883 2>,<at 1 1 RCIM_Cubic>]}
         GreenCurve {keys [<at 0 0.1578 RCIM_Cubic>,<at 0.3883 0.0017>,<at 1 0 RCIM_Cubic>]}
         RedCurve {keys [<at 0 3.2685 RCIM_Cubic>,<at 0.3883 0.5194>,<at 1 0.0277 RCIM_Cubic>]}
        }})
      Float=$System.ToDBrightness
     )
     Float=&FloatFromCurve(FloatCurve=DataInterface() {propertyValues {Curve {
         defaultValue=1
         keys [{interpMode=RCIM_Cubic, tangentMode=RCTM_User, time=0, value=1}, {interpMode=RCIM_Cubic, tangentMode=RCTM_User, time=1, value=0}]
        }}})
    )) {isSetParameters=true}
  ]
  rendererClasses [&NiagaraRibbonRendererProperties]
  renderers [
   {rendererClass=&NiagaraRibbonRendererProperties}
   {rendererClass=&NiagaraRibbonRendererProperties, rendererIndex=1}
  ]
 }}
summary {
 systemName=NS_Pickaxe_Vanceguard_Trail
 emitters [{emitterName=Trail, rendererClasses [&NiagaraRibbonRendererProperties]}]
 userVariables [
  User.bIsFrontEnd {type=NiagaraBool}
  User.bIsFrontEndPreview {type=NiagaraBool}
  User.RibbonIDOffset {type=NiagaraInt32}
 ]
}
